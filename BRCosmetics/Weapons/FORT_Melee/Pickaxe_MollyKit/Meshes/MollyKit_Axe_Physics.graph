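# Physics asset for the MollyKit pickaxe, a two-handed melee cosmetic skeletal mesh.
# A sphere body on the handle bone connects to a capsule body on the dyn_Wrench_1 bone through a single constraint.
# The constraint locks twist and swing 1 and leaves swing 2 as the angular axis not locked, with a narrow limit, so the wrench pivots in one plane relative to the handle.
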PhysicsAsset /BRCosmetics/Weapons/FORT_Melee/Pickaxe_MollyKit/Meshes/MollyKit_Axe_Physics
bodies {
 Melee_TwoHanded_Handle [{shapeType=Sphere, radius=4, shapeName=Melee_TwoHanded_Handle_sphere}]
 dyn_Wrench_1 [{shapeType=Capsule, length=29.7715, radius=4.2157, center=<8.3028 -0.5502 -17.7791>}]
}
constraints [{
  bone1Name=dyn_Wrench_1
  bone2Name=Melee_TwoHanded_Handle
  swing1LimitDegrees=45
  swing1Motion=Locked
  swing2LimitDegrees=20
  twistLimitDegrees=45
  twistMotion=Locked
 }]
